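# Material function that draws a two-digit counter from a number parameter and a digit atlas texture, on a backpack material.
# UVs from a selectable channel are rotated, offset and scaled about center, then split into two horizontal cells.
# Fmod separates tens and units; rounded U picks which digit offsets the atlas lookup along V.
# Sine-based edge masks with adjustable kerning and the scaled-UV 0-1 mask trim the output.
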MaterialFunction /BRCosmetics/Accessories/FORT_Backpacks/Backpack_M_MED_Tactical_Woodland_Blue/Materials/MF_counter_complex
graph {nodes [
  n1 = TextureCoordinate()
  n2 = TextureCoordinate() {coordinateIndex=1}
  n3 = ScalarParameter() {defaultValue=1, group=Counter, parameterName=CounterUseSecondUVChannel}
  n4 = Add(
   A=MaterialFunctionCall(
    UVs=LinearInterpolate(A=@n1, B=@n2, Alpha=@n3)
    Rotation Center=Constant2Vector() {g=0.5, r=0.5}
    "Rotation Angle (0-1)"=ScalarParameter() {group=Counter, parameterName=Counter_Rotation}
   ) {materialFunction=&CustomRotator}.Rotated Values
   B=AppendVector(A=ScalarParameter() {group=Counter, parameterName=U_pos}, B=ScalarParameter() {group=Counter, parameterName=V_pos})
  )
  n5 = MaterialFunctionCall(
   UVs=@n4
   Texture Scale=MaterialFunctionCall(
    X=ScalarParameter() {defaultValue=1, group=Counter, parameterName=CounterScaleU}
    Y=ScalarParameter() {defaultValue=1, group=Counter, parameterName=CounterScaleV}
   ) {materialFunction=&MakeFloat2}.Result
  ) {materialFunction=&ScaleUVsByCenter}
  n6 = Frac(None=@n5.UVs)
  n7 = Constant2Vector() {g=1, r=2}
  n8 = Add(A=Frac(None=Multiply(A=@n6, B=@n7))) {constB=-0.25}
  n9 = Sine(None=ComponentMask(None=@n8) {r=true})
  n10 = ScalarParameter() {defaultValue=0.02, group=Counter - Kerning, parameterName=KerningEdgeMaskSize}
  n11 = Ceil(None=Subtract(A=ConstantBiasScale(None=@n9), B=@n10) {constB=0.1})
  n12 = Sine(None=ComponentMask(None=@n8) {g=true})
  n13 = Ceil(None=Subtract(A=ConstantBiasScale(None=@n12), B=@n10) {constB=0.1})
  n14 = Clamp(None=Multiply(A=Multiply(A=@n11, B=@n13)) {constB=5})
  n15 = MaterialFunctionCall(Float2=Multiply(A=@n5.UVs, B=@n7)) {materialFunction=&BreakOutFloat2Components}
  n16 = Multiply(A=Frac(None=@n15.G)) {constB=0.1}
  n17 = ScalarParameter() {defaultValue=28, desc=input number, group=Counter, parameterName=Counter_InputNumber}
  n18 = Constant() {r=10}
  n19 = Fmod(A=@n17, B=@n18)
  n20 = Subtract(A=@n17, B=@n19)
  n21 = ComponentMask(None=@n4) {r=true}
  n22 = Divide(
   A=LinearInterpolate(A=Divide(A=@n20, B=@n18) {constB=10}, B=@n19, Alpha=Round(None=@n21)) {desc=left size is tens. right side is units}
   B=ScalarParameter() {defaultValue=10, desc=Number of Total Digits in Texture, group=Counter, parameterName=numColumns}
  ) {desc="multiply by how many digits in the texture.\r\nThis gives us an offset."}
  n23 = TextureSampleParameter2D(UVs=MaterialFunctionCall(X=@n15.R, Y=Add(A=@n16, B=@n22)) {materialFunction=&MakeFloat2}.Result) {
   group=Counter
   parameterName=Counter_AtlasTexture
   texture=/Game/Accessories/FORT_Backpacks/Backpack_F_MED_Cotton_Candy/Textures/T_NumberFlipBook
  }
  FunctionOutput(None=Multiply(A=@n5.0-1 mask, B=Multiply(A=@n14, B=@n23.RGB)))
  TextureCoordinate() {coordinateIndex=1}
 ]}
