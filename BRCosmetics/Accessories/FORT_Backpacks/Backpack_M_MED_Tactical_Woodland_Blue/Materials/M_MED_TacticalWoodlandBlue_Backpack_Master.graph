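# Master material for the Tactical Woodland Blue backpack, adding an animated counter screen to the shared character skin shading.
# Character PreFX, SkinFX and PostFX functions wrap an emissive override masked by the backpack FX texture.
# Its blue channel isolates the tinted counter, layered with procedural frame, scanlines, sweep and panning static.
# Its green channel isolates pulsing skin emissive, and a TimeStart-driven timestamp flash modulates both.
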Material /BRCosmetics/Accessories/FORT_Backpacks/Backpack_M_MED_Tactical_Woodland_Blue/Materials/M_MED_TacticalWoodlandBlue_Backpack_Master
graph {
 properties {
  bUseMaterialAttributes=true
  bUsedWithMeshDeformer=true
  bUsedWithSkeletalMesh=true
  pixelDepthOffsetMode=PDOM_Legacy
  refractionMethod=RM_IndexOfRefraction
  shadingModel=MSM_FromMaterialExpression
  usedShadingModels=Default Lit | Subsurface | Clear Coat | Subsurface Profile | Cloth | Eye
 }
 nodes [
  n1 = TextureCoordinate()
  n2 = VectorParameter() {parameterName=TexCoordScale, defaultValue=<rgba 4 4 0 0>}
  Add(
   A=Multiply(A=@n1, B=ComponentMask(None=@n2.RGB) {g=true, r=true})
   B=ComponentMask(None=VectorParameter() {parameterName=TexCoordOffset, defaultValue=<rgba 0.0524 -1.7822 0 0>}.RGB) {g=true, r=true}
  )
  n3 = Multiply(A=ComponentMask(None=NamedRerouteUsage()) {g=true}) {constB=0.75}
  n4 = Multiply(A=Time()) {constB=-0.2}
  n5 = Power(Base=Frac(None=Add(A=@n3, B=@n4))) {constExponent=20}
  AppendVector(A=Multiply(A=@n5, B=ScalarParameter() {defaultValue=0.005, group=Sweep Line, parameterName=Sweep UV Distort}) {constB=0.02}, B=Constant())
  n6 = NamedRerouteUsage()
  n7 = Multiply(
   A=TextureSample(UVs=Panner(Coordinate=@n6) {speedX=6, speedY=-0.05}) {samplerType=SAMPLERTYPE_Masks, texture=/Game/Missions/Primary/RepairTheShelter/Textures/Cinematic_Screen_Static}.G
   B=TextureSample(UVs=Panner(Coordinate=@n6) {speedX=-3, speedY=-0.03}) {samplerType=SAMPLERTYPE_Masks, texture=&Cinematic_Screen_Static}.R
  )
  AppendVector(A=Multiply(A=@n7, B=ScalarParameter() {defaultValue=0.07, group=Static, parameterName=Static UV Distort}), B=Constant())
  n8 = CameraVectorWS()
  n9 = VertexNormalWS()
  n10 = TextureCoordinate()
  n11 = MaterialFunctionCall(Y=Divide(A=Time()) {constB=-6}) {materialFunction=&MakeFloat2}
  n12 = Multiply(A=ComponentMask(None=Add(A=@n10, B=@n11.Result)) {g=true})
  Clamp(None=Multiply(A=DotProduct(A=@n8, B=@n9), B=Sine(None=@n12)))
  n13 = Constant() {r=0.5}
  LinearInterpolate(
   A=MaterialFunctionCall!(ScanLine_Tiling_1, Scanline_PannerSpeed_1, ScanLine_Tiling, Scanline_PannerSpeed, UVs) {materialFunction=/Game/Weapons/WeaponSkins/MaterialFunction/ScanLines}.Result
   B=@n13
   Alpha=MaterialFunctionCall!(Blend Range=ScalarParameter() {defaultValue=25, parameterName=r}, Start Offset=ScalarParameter() {defaultValue=-100, parameterName=d}) {materialFunction=/Engine/Functions/Engine_MaterialFunctions03/Particles/Distance_Blend}.Result
  )
  NamedRerouteDeclaration(None=@n1) {name=TexCoord, nodeColor=<rgba 1 0 0.4453 1>}
  n14 = MaterialFunctionCall() {materialFunction=&BaseFN_CharacterFunction_PreFX}
  n15 = GetMaterialAttributes(None=MaterialFunctionCall(
    PreskinnedPosition=@n14.Pre skinned Local Position
    Gmap=@n14.Pre skinned Local Position
    MaterialAttributes=@n14.Pre skinned Local Position
    AO=@n14.Pre skinned Local Position
   ) {materialFunction=&CharacterSkinFX}.Result) {attributeGetTypes [#1, #2]}
  n16 = TextureSample() {texture=/BRCosmetics/Accessories/FORT_Backpacks/Backpack_M_MED_Tactical_Woodland_Blue/Textures/T_M_MED_Tactical_Woodland_Blue_Backpack_FX}
  n17 = LinearInterpolate(A=@n15.Emissive Color, B=@n15.Emissive Color, Alpha=Add(A=@n16.G, B=@n16.G)) {constB=0, desc=Allow Other Emissive Through}
  n18 = LinearInterpolate(B=@n15.Emissive Color, Alpha=@n16.G) {desc=IsolateScreenEmissive}
  n19 = Divide(A=Time()) {constB=10}
  n20 = Multiply(A=ScalarParameter() {defaultValue=4, group=Background, parameterName=EffectPulseMultiplier}, B=Abs(None=Sine(None=@n19))) {constB=4}
  n21 = Multiply(A=NamedRerouteUsage()) {constB=0.5}
  n22 = Multiply(A=Sine(None=ComponentMask(None=@n21) {r=true}), B=Sine(None=ComponentMask(None=@n21) {g=true}))
  n23 = OneMinus(None=Power(Base=OneMinus(None=@n22), Exp=ScalarParameter() {defaultValue=0.8, group=Background, parameterName=Frame Exponent}))
  n24 = NamedRerouteUsage()
  n25 = ConstantBiasScale(None=Frac(None=Panner(Coordinate=@n24) {speedY=-1})) {bias=-0.5, scale=2}
  n26 = Multiply(A=@n25, B=@n25)
  n27 = Add(A=Multiply(A=ComponentMask(None=@n26) {g=true}, B=ScalarParameter() {defaultValue=0.5, group=Background, parameterName=Scanline Strength}) {constB=0.1})
  n28 = ScalarParameter() {defaultValue=0.2, group=Sweep Line, parameterName=Sweep Add}
  n29 = Panner(Coordinate=@n6) {speedX=5, speedY=0.1}
  n30 = Add(
   A=Multiply(A=TextureSample(UVs=@n29) {samplerType=SAMPLERTYPE_Masks, texture=&Cinematic_Screen_Static}.B) {constB=0.1}
   B=@n7
  )
  n31 = ScalarParameter() {defaultValue=6, group=Static, parameterName=Static Scale}
  n32 = Add(
   A=Add(A=Multiply(A=@n30, B=@n31), B=Multiply(A=@n5, B=@n28) {constB=0.2})
   B=ScalarParameter() {defaultValue=0.2, group=Background, parameterName=Background}
  )
  n33 = Saturate(None=ComponentMask(None=MaterialFunctionCall() {materialFunction=&/MF_counter_complex}.Result) {r=true})
  n34 = Add(A=Multiply(A=Multiply(A=@n13, B=@n33), B=ScalarParameter() {defaultValue=5, parameterName=FaceIntensity})) {constB=0}
  n35 = Multiply(
   A=Multiply(
    A=VectorParameter() {group=Counter, parameterName=Counter Color, defaultValue=<rgba 0.8229 0.4209 0 0>}.RGB
    B=ScalarParameter() {defaultValue=1, group=Counter, parameterName=CounterBrightness}
   )
   B=Multiply(A=Multiply(A=@n23, B=@n27), B=Add(A=@n32, B=@n34)) {constA=1}
  )
  n36 = Sine(None=MaterialFunctionCall(
    TimestampA=ScalarParameter() {parameterName=TimeStart}
    TransitionLength_In=ScalarParameter() {defaultValue=1, group=Counter, parameterName=FlashLength}
    Transition In Only?=StaticBool() {value=true}
   ) {materialFunction=&MF_Timestamp_InOut_01}.Linear) {period=2}
  n37 = Multiply(A=Add(A=Multiply(A=@n18, B=@n20), B=LinearInterpolate(B=@n35, Alpha=@n16.B) {desc=IsolateScreen}), B=Add(A=Multiply(A=@n36) {constB=20}))
  n38 = MaterialFunctionCall(
   AO=@n14.AO
   Distance Blend=@n14.AO
   Fresnel=@n14.AO
   Gmap=@n14.AO
   Pre Skinned Local Position=@n14.AO
   Material=SetMaterialAttributes!(MaterialAttributes=@n15.MaterialAttributes, Emissive Color=Add(A=@n17, B=@n37), Base Color=@n15.MaterialAttributes) {attributeSetTypes [#1, #2]}
  ) {materialFunction=&BaseFN_CharacterFunction_PostFX}
  SetMaterialAttributes!(
   MaterialAttributes=GetMaterialAttributes(None=@n38.Result) {attributeGetTypes [#1]}.MaterialAttributes
   Base Color
   Emissive Color
  ) {attributeSetTypes [#2, #1]}
  TextureCoordinate() {vTiling=15}
  TextureCoordinate()
 ]
}
